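# Unlit masked mesh-particle material for the flame effect on the Horned Mask glider.
# A sphere mask around the particle position, sized by dynamic parameters, shapes both panning vector-noise displacement and glow, with a mobile fallback replacing the mask.
# A time-driven sine adds vertical bob.
# Diffuse blue channel and mask blend two emissive colors tinted by particle color, while particle alpha and tiled noise form a dissolve value.
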Material /BRCosmetics/Gadgets/Assets/VinderTech_GliderChute/Glider_HornedMask/Materials/M_Glider_HornedMask_Flames
graph {
 properties {bUsedWithMeshParticles=true, blendMode=BLEND_Masked, pixelDepthOffsetMode=PDOM_Legacy, shadingModel=MSM_Unlit, usedShadingModels=Unlit}
 nodes [
  n1 = SphereMask(
   A=WorldPosition().XYZ
   B=ParticlePositionWS()
   Radius=DynamicParameter() {defaultValue=<rgba 1 25 1 1>, paramNames [RandomNoise, SphereMaskSize, RandomBob, Param4]}.SphereMaskSize
   Hardness=Constant()
  )
  n2 = ShadingPathSwitch(Default=@n1, Deferred=@n1, Mobile=Constant() {r=1}) {desc=Migrated from FeatureLevelSwitch}
  n3 = Clamp(None=Power(Base=OneMinus(None=@n2)))
  n4 = Add(
   A=DynamicParameter() {defaultValue=<rgba 1 25 1 1>, paramNames [RandomNoise, SphereMaskSize, RandomBob, Param4]}.RandomNoise
   B=WorldPosition().XYZ
  )
  n5 = Multiply(A=Time(), B=ScalarParameter() {defaultValue=20, parameterName=PanSpeed})
  n6 = VectorNoise(World Position=Multiply(A=Add(A=@n4, B=@n5), B=ScalarParameter() {defaultValue=0.005, parameterName=NoiseTile}) {constB=0.01}) {noiseFunction=VNF_VectorALU, tileSize=1}
  n7 = ScalarParameter() {defaultValue=8, parameterName=NoiseScale}
  n8 = Add(A=DynamicParameter() {defaultValue=<rgba 1 25 1 1>, paramNames [RandomNoise, SphereMaskSize, RandomBob, Param4]}.RandomBob, B=Time())
  Add(
   A=Multiply(A=LinearInterpolate(Alpha=@n3), B=Multiply(A=@n6, B=@n7))
   B=Multiply(A=Sine(None=@n8) {period=2}, B=Constant3Vector() {constant=<rgba 0 0 2.5 0>})
  )
  n9 = ParticleColor()
  n10 = TextureSample(UVs=TextureCoordinate() {uTiling=10, vTiling=10}) {texture=&T_NoisePack_01}
  Clamp(None=Add(A=LinearInterpolate(Alpha=@n9.A) {constA=-1}, B=Multiply(A=@n10.G)))
  n11 = TextureSample() {samplerType=SAMPLERTYPE_Normal, texture=/BRCosmetics/Gadgets/Assets/VinderTech_GliderChute/Glider_HornedMask/Textures/T_Glider_HornedMask_N}
  Fresnel(Normal=Transform(None=@n11.RGB)) {exponent=1}
  MaterialFunctionCall() {materialFunction=&DebugTimeSine}
  n12 = VectorParameter() {parameterName=ColorA, defaultValue=<rgba 0.0019 0.0137 0.1 1>}
  n13 = VectorParameter() {parameterName=ColorB, defaultValue=<rgba 0.2292 1.2347 4 1>}
  n14 = Multiply(
   A=LinearInterpolate(Alpha=TextureSample() {texture=/BRCosmetics/Gadgets/Assets/VinderTech_GliderChute/Glider_HornedMask/Textures/T_Glider_HornedMask_D}.B) {constA=0.1}
   B=@n2
  )
  MaterialFunctionCall(Emissive=Multiply(A=LinearInterpolate(A=@n12.RGB, B=@n13.RGB, Alpha=@n14), B=@n9.RGB)) {materialFunction=&MF_FX_EmissiveAdjustments}
 ]
}
